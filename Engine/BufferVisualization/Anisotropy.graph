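# Unlit post-process material for the engine's Anisotropy buffer visualization, drawn after depth of field with pre-exposure scaling disabled.
# It reads the red channel of the Anisotropy scene texture and splits it by sign, putting saturated positive values in green and negated negative values in blue.
# The packed color is raised to a 2.2 power.
# A shading path switch outputs a zero constant under forward shading.
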Material /Engine/BufferVisualization/Anisotropy
graph {
 properties {
  bDisablePreExposureScale=true
  blendableLocation=BL_SceneColorAfterDOF
  materialDomain=MD_PostProcess
  pixelDepthOffsetMode=PDOM_Legacy
  previewMesh=/Engine/Transient.MaterialEditorMeshComponent_3
  shadingModel=MSM_Unlit
  usedShadingModels=Unlit
 }
 nodes [
  n1 = ComponentMask(None=SceneTexture() {sceneTextureId=PPI_Anisotropy}.Color) {r=true}
  n2 = Multiply(A=@n1) {constB=-1}
  n3 = Power(Base=MaterialFunctionCall(G=Saturate(None=@n1), B=Saturate(None=@n2)) {materialFunction=&AppendMany}.RGB) {constExponent=2.2}
  ShadingPathSwitch(Default=FeatureLevelSwitch(Default=@n3, ES2=@n3, ES3_1=@n3, SM4_REMOVED=@n3), Forward=Constant())
 ]
}
