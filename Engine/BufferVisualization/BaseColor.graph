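# Post-process material for the BaseColor buffer visualization, displaying the stored base color of the scene.
# A SceneTexture node samples PPI_StoredBaseColor and feeds every listed input of a feature level switch.
# A shading path switch passes that color through by default and substitutes a constant at its default value on the Forward path.
# It is placed after depth of field with pre-exposure scaling disabled.
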Material /Engine/BufferVisualization/BaseColor
graph {
 properties {
  bDisablePreExposureScale=true
  blendableLocation=BL_SceneColorAfterDOF
  materialDomain=MD_PostProcess
  pixelDepthOffsetMode=PDOM_Legacy
  previewMesh=/Engine/Transient.MaterialEditorMeshComponent_5
  shadingModel=MSM_Unlit
  usedShadingModels=Unlit
 }
 nodes [
  Constant() {r=1}
  n1 = SceneTexture() {sceneTextureId=PPI_StoredBaseColor}
  ShadingPathSwitch(Default=FeatureLevelSwitch(Default=@n1.Color, ES2=@n1.Color, ES3_1=@n1.Color, SM4_REMOVED=@n1.Color), Forward=Constant())
 ]
}
