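# Unlit post-process material that visualizes the custom depth buffer for buffer visualization.
# It samples custom depth through SceneTexture, scales it down and applies Frac, so the value repeats as a gradient every 10,000 units instead of clamping.
# The result feeds a FeatureLevelSwitch whose listed inputs all share that same value.
# Pre-exposure scaling is disabled on the material.
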Material /Engine/BufferVisualization/CustomDepth
graph {
 properties {
  bDisablePreExposureScale=true
  materialDomain=MD_PostProcess
  pixelDepthOffsetMode=PDOM_Legacy
  previewMesh=/Engine/Transient.MaterialEditorMeshComponent_7
  shadingModel=MSM_Unlit
  usedShadingModels=Unlit
 }
 nodes [
  n1 = Frac(None=Multiply(A=SceneTexture() {sceneTextureId=PPI_CustomDepth}.Color) {constB=0.0001})
  FeatureLevelSwitch(Default=@n1, ES2=@n1, ES3_1=@n1, SM4_REMOVED=@n1)
 ]
}
