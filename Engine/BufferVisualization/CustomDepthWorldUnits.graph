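# Unlit post-process material for buffer visualization that shows Custom Depth in world units.
# A Custom node transforms remapped viewport UVs through ScreenToWorld into a view ray, which is multiplied by a masked channel of the CustomDepth scene texture.
# The resulting vector's length is scaled by 0.01, converting centimeters to meters, and feeds a feature level switch with identical branches.
# Pre-exposure scaling is disabled.
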Material /Engine/BufferVisualization/CustomDepthWorldUnits
graph {
 properties {
  bDisablePreExposureScale=true
  materialDomain=MD_PostProcess
  pixelDepthOffsetMode=PDOM_Legacy
  previewMesh=/Engine/Transient.MaterialEditorMeshComponent_9
  shadingModel=MSM_Unlit
  usedShadingModels=Unlit
 }
 nodes [
  n1 = ComponentMask(None=SceneTexture() {sceneTextureId=PPI_CustomDepth}.Color) {a=true}
  MaterialExpressionSubtract_0 = Subtract(A=Multiply(A=ScreenPosition().ViewportUV) {constB=2})
  n2 = Custom!(ScreenPosition=@MaterialExpressionSubtract_0) {
   code="mul(float4(ScreenPosition.xy, 1, 0), LWCToFloat(ResolvedView.ScreenToWorld)).xyz"
   desc="\n*** Original source before expression upgrade ***\nmul(float4(ScreenPosition.xy, 1, 0), View.ScreenToWorld).xyz"
   inputs [{inputName=ScreenPosition, input {expression=&self:MaterialExpressionSubtract_0}}]
  }
  n3 = Multiply(A=MaterialFunctionCall(Vector 3=Multiply(A=@n1, B=@n2)) {materialFunction=/Engine/Functions/Engine_MaterialFunctions02/Utility/VectorLength}.V3 Length) {constB=0.01}
  FeatureLevelSwitch(Default=@n3, ES2=@n3, ES3_1=@n3, SM4_REMOVED=@n3)
 ]
}
