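# Unlit post-process material for the Custom Stencil buffer visualization, showing each stencil value as a color with its number printed on top.
# Custom hash nodes turn the stencil value into pseudo-random RGB, masked to black where stencil is zero.
# Screen-aligned UVs tile a three-digit text box, sampling thresholded MiniFont glyphs for each decimal digit.
# The digits blend over the color, which then multiplies the result and brightens it.
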Material /Engine/BufferVisualization/CustomStencil
graph {
 properties {
  bDisablePreExposureScale=true
  materialDomain=MD_PostProcess
  pixelDepthOffsetMode=PDOM_Legacy
  previewMesh=/Engine/Transient.MaterialEditorMeshComponent_11
  shadingModel=MSM_Unlit
  usedShadingModels=Unlit
 }
 nodes [
  n1 = ComponentMask(None=SceneTexture() {sceneTextureId=PPI_CustomStencil}.Color) {r=true}
  MaterialExpressionDivide_0 = Divide(A=@n1) {constB=256}
  n2 = Custom!(idx=@MaterialExpressionDivide_0) {
   code="frac(dot(float3((idx.x + 33/ 2) - 64.340622f, -72.465622f, (idx.x + 33 / 2) - 64.340622f) * float3((idx.x  + 33/ 2) - 64.340622f, -72.465622f, -72.465622f), float3(20.390625f, 60.703125f, 2.4281209f)))"
   description=Random R
   outputType=CMOT_Float1
   inputs [{inputName=idx, input {expression=&self:MaterialExpressionDivide_0}}]
  }
  n3 = Custom(idx=@MaterialExpressionDivide_0) {
   code="frac(dot(float3((idx.x + 1/ 2) - 64.340622f, -72.465622f, (idx.x  + 1/ 2) - 64.340622f) * float3((idx.x + 1 / 2) - 64.340622f, -72.465622f, -72.465622f), float3(20.390625f, 60.703125f, 2.4281209f)))"
   description=Random G
   outputType=CMOT_Float1
   inputs [{inputName=idx, input {expression=&self:MaterialExpressionDivide_0}}]
  }
  n4 = Custom(idx=@MaterialExpressionDivide_0) {
   code="frac(dot(float3((idx.x + 71/ 2) - 64.340622f, -72.465622f, (idx.x  + 71/ 2) - 64.340622f) * float3((idx.x + 71 / 2) - 64.340622f, -72.465622f, -72.465622f), float3(20.390625f, 60.703125f, 2.4281209f)))"
   description=Random B
   outputType=CMOT_Float1
   inputs [{inputName=idx, input {expression=&self:MaterialExpressionDivide_0}}]
  }
  n5 = MaterialFunctionCall(X=@n2, Y=@n3, Z=@n4) {materialFunction=&MakeFloat3}
  n6 = Clamp(None=@n5.Result) {minDefault=0.001}
  MaterialExpressionMultiply_0 = Multiply(A=Clamp(None=@n1), B=LinearInterpolate(A=Normalize(VectorInput=@n6), B=@n5.Result) {constAlpha=0.65})
  MaterialExpressionMaterialFunctionCall_5 = MaterialFunctionCall() {materialFunction=&ScreenAlignedUVs}
  MaterialExpressionCustom_13 = Custom!(a=@MaterialExpressionMaterialFunctionCall_5."X 100%, Y 100%") {
   code="frac(a.xy * 20)"
   description=Stencil Text Tiling
   outputType=CMOT_Float2
   inputs [{inputName=a, input {expression=&self:MaterialExpressionMaterialFunctionCall_5}}]
  }
  MaterialExpressionCustom_9 = Custom(a=@MaterialExpressionCustom_13) {
   code="((a.x - 0.25f) / 0.5f) * 3.f"
   description=Text U Coord
   outputType=CMOT_Float1
   inputs [{inputName=a, input {expression=&self:MaterialExpressionCustom_13}}]
  }
  MaterialExpressionComponentMask_1 = ComponentMask(None=SceneTexture() {sceneTextureId=PPI_CustomStencil}.Color) {r=true}
  MaterialExpressionCustom_8 = Custom!(a=@MaterialExpressionCustom_9, b=@MaterialExpressionComponentMask_1) {
   code=((int)trunc((a < 1.f) ? (b / 100) : ((a < 2.f) ? (b / 10) : b)) % 10)
   description=Unit Value
   outputType=CMOT_Float1
   inputs [
    {inputName=a, input {expression=&self:MaterialExpressionCustom_9}}
    {inputName=b, input {expression=&self:MaterialExpressionComponentMask_1}}
   ]
  }
  n7 = Custom(a=@MaterialExpressionCustom_8, b=@MaterialExpressionCustom_9) {
   code="lerp(a / 64.f, (a + 1) / 64.f, b - trunc(b))"
   description=Letter U Coord
   outputType=CMOT_Float1
   inputs [
    {inputName=a, input {expression=&self:MaterialExpressionCustom_8}}
    {inputName=b, input {expression=&self:MaterialExpressionCustom_9}}
   ]
  }
  MaterialExpressionCustom_7 = Custom!(a=@MaterialExpressionCustom_13) {
   code="(a.y - 0.3) / (0.7 - 0.3)"
   description=Letter V Coord
   outputType=CMOT_Float1
   inputs [{inputName=a, input {expression=&self:MaterialExpressionCustom_13}}]
  }
  MaterialExpressionTextureSample_0 = TextureSample!(
   UVs=MaterialFunctionCall(X=@n7, Y=@MaterialExpressionCustom_7) {materialFunction=&MakeFloat2}.Result
   Tex
   Level
   Apply View MipBias
  ) {constMipValue=0, mipValueMode=TMVM_MipLevel, texture=/Engine/EngineMaterials/MiniFont}
  n8 = Custom(a=@MaterialExpressionTextureSample_0.R) {
   code=(a > 0.5f ? 1.f : 0.f)
   description=Font Threshold
   outputType=CMOT_Float1
   inputs [{inputName=a, input {expression=&self:MaterialExpressionTextureSample_0, mask=1, maskR=1, outputIndex=1}}]
  }
  n9 = Custom!(a=@MaterialExpressionCustom_9, b=@MaterialExpressionCustom_7) {
   code="abs(a - 1.5) <= 1.5 &&&& abs(b - 0.5) <= 0.5"
   description=Text Box Mask
   outputType=CMOT_Float1
   inputs [
    {inputName=a, input {expression=&self:MaterialExpressionCustom_9}}
    {inputName=b, input {expression=&self:MaterialExpressionCustom_7}}
   ]
  }
  MaterialExpressionLinearInterpolate_1 = LinearInterpolate(A=@MaterialExpressionMultiply_0, B=@n8, Alpha=@n9) {constAlpha=0.65}
  n10 = Custom(a=@MaterialExpressionLinearInterpolate_1, b=@MaterialExpressionMultiply_0) {
   code=a * b * 1.5f
   description=Brighten
   inputs [
    {inputName=a, input {expression=&self:MaterialExpressionLinearInterpolate_1}}
    {inputName=b, input {expression=&self:MaterialExpressionMultiply_0}}
   ]
  }
  FeatureLevelSwitch(Default=@n10, ES2=@n10, ES3_1=@n10, SM4_REMOVED=@n10)
  MaterialExpressionMaterialFunctionCall_1 = MaterialFunctionCall() {materialFunction=&ScreenAlignedUVs}
  n11 = Custom!(a=@MaterialExpressionMaterialFunctionCall_1."X 100%, Y 100%") {
   code=((a.x > 0.485) &&&& (a.x < 0.515) &&&& (a.y > 0.52)  &&&& (a.y < 0.535) ? 1.f : 1.f)
   description=Display Area
   outputType=CMOT_Float1
   inputs [{inputName=a, input {expression=&self:MaterialExpressionMaterialFunctionCall_1}}]
  }
  Multiply(A=Clamp(None=@MaterialExpressionComponentMask_1), B=@n11)
 ]
}
