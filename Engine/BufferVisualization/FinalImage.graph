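# FinalImage is an unlit post-process material among the engine's buffer visualization materials, representing the final image view.
# It reads the color of the PostProcessInput0 scene texture with no further processing.
# A feature level switch receives that same color on its Default, ES2, ES3_1 and removed SM4 inputs, so those paths are identical.
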Material /Engine/BufferVisualization/FinalImage
graph {
 properties {materialDomain=MD_PostProcess, pixelDepthOffsetMode=PDOM_Legacy, shadingModel=MSM_Unlit, usedShadingModels=Unlit}
 nodes [
  n1 = SceneTexture() {sceneTextureId=PPI_PostProcessInput0}
  FeatureLevelSwitch(Default=@n1.Color, ES2=@n1.Color, ES3_1=@n1.Color, SM4_REMOVED=@n1.Color)
 ]
}
