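# Unlit post-process buffer visualization material that shows scene depth in world units as distance from the camera.
# It takes the length of the camera-relative world position, so depth is radial rather than planar, and scales it from centimeters to meters.
# A feature level switch substitutes a constant white value on ES2 and ES3.1.
# Pre-exposure scaling is disabled on the material.
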Material /Engine/BufferVisualization/SceneDepthWorldUnits
graph {
 properties {
  bDisablePreExposureScale=true
  materialDomain=MD_PostProcess
  pixelDepthOffsetMode=PDOM_Legacy
  previewMesh=/Engine/Transient.MaterialEditorMeshComponent_33
  shadingModel=MSM_Unlit
  usedShadingModels=Unlit
 }
 nodes [
  n1 = Multiply(A=MaterialFunctionCall(Vector 3=WorldPosition() {worldPositionShaderOffset=WPT_CameraRelative}.XYZ) {materialFunction=/Engine/Functions/Engine_MaterialFunctions02/Utility/VectorLength}.V3 Length) {constB=0.01}
  n2 = Constant() {r=1}
  FeatureLevelSwitch(Default=@n1, ES2=@n2, ES3_1=@n2, SM4_REMOVED=@n1)
 ]
}
