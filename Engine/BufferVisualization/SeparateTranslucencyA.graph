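# Post-process material that visualizes the alpha channel of the separate translucency buffer, as an engine buffer visualization view.
# It samples PostProcessInput1 and masks out its alpha, blending in before bloom with pre-exposure scaling disabled.
# A feature level switch replaces that sample with a scalar parameter left at its default on ES2 and ES3.1.
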Material /Engine/BufferVisualization/SeparateTranslucencyA
graph {
 properties {
  bDisablePreExposureScale=true
  blendableLocation=BL_SceneColorBeforeBloom
  materialDomain=MD_PostProcess
  pixelDepthOffsetMode=PDOM_Legacy
  previewMesh=/Engine/Transient.MaterialEditorMeshComponent_35
  shadingModel=MSM_Unlit
  usedShadingModels=Unlit
 }
 nodes [
  n1 = ComponentMask(None=SceneTexture() {sceneTextureId=PPI_PostProcessInput1}.Color) {a=true}
  n2 = ScalarParameter() {parameterName=0}
  FeatureLevelSwitch(Default=@n1, ES2=@n2, ES3_1=@n2, SM4_REMOVED=@n1)
 ]
}
