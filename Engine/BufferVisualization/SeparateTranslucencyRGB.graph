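# Post-process material for buffer visualization that displays the RGB color of the separate translucency buffer.
# It runs at the scene-color-after-DOF blendable location and samples PostProcessInput1 through a SceneTexture node.
# A feature level switch passes that color through by default and substitutes a constant one (white) on ES2 and ES3.1.
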Material /Engine/BufferVisualization/SeparateTranslucencyRGB
graph {
 properties {
  blendableLocation=BL_SceneColorAfterDOF
  materialDomain=MD_PostProcess
  pixelDepthOffsetMode=PDOM_Legacy
  shadingModel=MSM_Unlit
  usedShadingModels=Unlit
 }
 nodes [
  n1 = SceneTexture() {sceneTextureId=PPI_PostProcessInput1}
  n2 = Constant() {r=1}
  FeatureLevelSwitch(Default=@n1.Color, ES2=@n2, ES3_1=@n2, SM4_REMOVED=@n1.Color)
 ]
}
