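# Post-process material that draws the engine's Velocity buffer visualization.
# It divides SceneTexture velocity by a constant, offsets it to a mid-range bias, then builds a color with the R component feeding both X and Y, so the second velocity component is unused.
# Feature level and shading path switches substitute constants for mobile levels and forward shading.
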Material /Engine/BufferVisualization/Velocity
graph {
 properties {
  bDisablePreExposureScale=true
  blendableOutputAlpha=true
  materialDomain=MD_PostProcess
  pixelDepthOffsetMode=PDOM_Legacy
  previewMesh=/Engine/Transient.MaterialEditorMeshComponent_43
  shadingModel=MSM_Unlit
  usedShadingModels=Unlit
 }
 nodes [
  Constant()
  n1 = SceneTexture() {sceneTextureId=PPI_Velocity}
  n2 = MaterialFunctionCall(Float2=Add(A=Divide(A=@n1.Color) {constB=4}) {constB=0.5}) {materialFunction=&BreakOutFloat2Components}
  n3 = MaterialFunctionCall(X=@n2.R, Y=@n2.R) {materialFunction=&MakeFloat3}
  n4 = Constant()
  ShadingPathSwitch(Default=FeatureLevelSwitch(Default=@n3.Result, ES2=@n4, ES3_1=@n4, SM4_REMOVED=@n3.Result), Forward=Constant())
 ]
}
